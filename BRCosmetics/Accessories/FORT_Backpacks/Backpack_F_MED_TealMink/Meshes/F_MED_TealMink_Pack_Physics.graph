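# Physics asset for the TealMink backpack cosmetic, defining collision bodies and joints for a small character's dyn_ bones.
# Capsules cover torso, head, ears, spin, and two-segment arms and legs, with a second small capsule on each lower arm.
# Constraints branch from the torso to head, upper arms and spin, then to ears, lower arms and legs.
# Angular limits are mostly tight, except the slightly asymmetric, wide-swinging shoulders.
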PhysicsAsset /BRCosmetics/Accessories/FORT_Backpacks/Backpack_F_MED_TealMink/Meshes/F_MED_TealMink_Pack_Physics
bodies {
 dyn_arm_1_l [{shapeType=Capsule, length=3, radius=1.5, shapeName=dyn_arm_1_l_capsule, center=<1.7109 0.6255 0>, rotation=<rot 90 160.96 170.9603>}]
 dyn_arm_1_r [{shapeType=Capsule, length=3, radius=1.5, shapeName=dyn_arm_1_r_capsule, center=<1.7459 -0.6556 0>, rotation=<rot 90 0 -10.0002>}]
 dyn_arm_2_l [
  {shapeType=Capsule, length=7, radius=1.4, shapeName=dyn_arm_2_l_capsule, center=<4.0982 0 0>, rotation=<rot 90 0 0>}
  {
   shapeType=Capsule
   length=2
   radius=0.8
   shapeName=dyn_arm_2_l_capsule
   center=<5.4849 -1.1511 -1.0828>
   rotation=<rot -70.1043 -166.0255 149.3366>
  }
 ]
 dyn_arm_2_r [
  {shapeType=Capsule, length=7, radius=1.4, shapeName=dyn_arm_2_r_capsule, center=<4.0143 0 0>, rotation=<rot 90 0 0>}
  {shapeType=Capsule, length=2, radius=0.8, shapeName=dyn_arm_2_r_capsule, center=<5.5 1.16 -1.3>, rotation=<rot -80 -103 122>}
 ]
 dyn_ear_1_l [{shapeType=Capsule, length=2, radius=2.5, shapeName=dyn_ear_1_l_capsule, center=<1.8385 0 -0.5855>, rotation=<rot 0 0 90>}]
 dyn_ear_1_r [{shapeType=Capsule, length=2, radius=2.5, shapeName=dyn_ear_1_r_capsule, center=<1.8385 0 -0.6919>, rotation=<rot 0 0 90>}]
 dyn_head [{shapeType=Capsule, length=0, radius=9, shapeName=dyn_head_capsule, center=<10.2706 0 0>}]
 dyn_leg_1_l [{shapeType=Capsule, length=1.5, radius=1.8, shapeName=dyn_leg_1_l_capsule, center=<1.2861 0.2298 -0.3051>, rotation=<rot 80.0001 0 0>}]
 dyn_leg_1_r [{shapeType=Capsule, length=1.5, radius=1.8, shapeName=dyn_leg_1_r_capsule, center=<1.2861 0 -0.3051>, rotation=<rot 80.0001 0 0>}]
 dyn_leg_2_l [{
   shapeType=Capsule
   length=3
   radius=2
   shapeName=dyn_leg_2_l_capsule
   center=<2.1798 0.1661 0.2921>
   rotation=<rot 84.9994 179.9999 179.9999>
  }]
 dyn_leg_2_r [{
   shapeType=Capsule
   length=3
   radius=2
   shapeName=dyn_leg_2_r_capsule
   center=<2.1798 -0.1342 0.2921>
   rotation=<rot 79.9981 -179.9998 -179.9998>
  }]
 dyn_spin [{shapeType=Capsule, length=1.3, radius=3.5, shapeName=dyn_spin_capsule, center=<2.3294 0 -0.4273>, rotation=<rot 90 179.9801 -90.0199>}]
 dyn_torso [{shapeType=Capsule, length=0, radius=3.5, shapeName=dyn_torso_capsule}]
}
constraints [
 {bone1Name=dyn_arm_2_l, bone2Name=dyn_arm_1_l, swing1LimitDegrees=5, swing2LimitDegrees=10, twistLimitDegrees=5}
 {bone1Name=dyn_arm_2_r, bone2Name=dyn_arm_1_r, swing1LimitDegrees=5, swing2LimitDegrees=10, twistLimitDegrees=5}
 {bone1Name=dyn_leg_2_l, bone2Name=dyn_leg_1_l, swing1LimitDegrees=5, swing2LimitDegrees=10, twistLimitDegrees=2}
 {bone1Name=dyn_leg_2_r, bone2Name=dyn_leg_1_r, swing1LimitDegrees=5, swing2LimitDegrees=10, twistLimitDegrees=2}
 {bone1Name=dyn_leg_1_l, bone2Name=dyn_spin, swing1LimitDegrees=5, swing2LimitDegrees=15, twistLimitDegrees=5}
 {bone1Name=dyn_leg_1_r, bone2Name=dyn_spin, swing1LimitDegrees=5, swing2LimitDegrees=15, twistLimitDegrees=5}
 {bone1Name=dyn_ear_1_l, bone2Name=dyn_head, swing1LimitDegrees=5, swing2LimitDegrees=15, twistLimitDegrees=1}
 {bone1Name=dyn_ear_1_r, bone2Name=dyn_head, swing1LimitDegrees=5, swing2LimitDegrees=15, twistLimitDegrees=1}
 {bone1Name=dyn_head, bone2Name=dyn_torso, swing1LimitDegrees=12, swing2LimitDegrees=7, twistLimitDegrees=7}
 {bone1Name=dyn_arm_1_l, bone2Name=dyn_torso, swing1LimitDegrees=40, swing2LimitDegrees=100, twistLimitDegrees=7}
 {bone1Name=dyn_arm_1_r, bone2Name=dyn_torso, swing1LimitDegrees=50, swing2LimitDegrees=100, twistLimitDegrees=5}
 {bone1Name=dyn_spin, bone2Name=dyn_torso, swing1LimitDegrees=10, swing2LimitDegrees=30, twistLimitDegrees=1}
]
